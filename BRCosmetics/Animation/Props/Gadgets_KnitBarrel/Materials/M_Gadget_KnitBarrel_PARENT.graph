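# Parent material for the KnitBarrel gadget prop, built on the 2023 CharacterParent PreFX and PostFX functions with animated tank-tread UV panning.
# An IsFrontEnd parameter blends between a TimeElasped parameter and Time, scaled by TreadsSpeed, with two masks XORed to flip direction.
# Step masks wrap the scrolled V coordinate within a texture band.
# A channel mask, blue by default, limits the override UVs fed into PreFX.
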Material /BRCosmetics/Animation/Props/Gadgets_KnitBarrel/Materials/M_Gadget_KnitBarrel_PARENT
graph {
 properties {
  bUseMaterialAttributes=true
  bUsedWithMeshDeformer=true
  bUsedWithMorphTargets=true
  bUsedWithNiagaraMeshParticles=true
  bUsedWithSkeletalMesh=true
  pixelDepthOffsetMode=PDOM_Legacy
  shadingModel=MSM_FromMaterialExpression
  usedShadingModels=Default Lit | Subsurface | Clear Coat | Subsurface Profile | Cloth
 }
 nodes [
  n1 = TextureCoordinate()
  n2 = MaterialFunctionCall(Float2=@n1) {materialFunction=&BreakOutFloat2Components}
  n3 = Step(Y=Constant() {r=0.5}, X=NamedRerouteDeclaration(None=@n2.G) {name=G, nodeColor=<rgba 0.1484 0 1 1>}) {constX=0.5, constY=0.4}
  n4 = LinearInterpolate(
   A=ScalarParameter() {group=TankTreadPanning, parameterName=TimeElasped}
   B=Time()
   Alpha=ScalarParameter() {group=TankTreadPanning, parameterName=IsFrontEnd}
  )
  n5 = ScalarParameter() {defaultValue=0.24, group=TankTreadPanning, parameterName=TreadsSpeed, sortPriority=0}
  n6 = NamedRerouteUsage()
  n7 = NamedRerouteUsage()
  n8 = Saturate(None=Add(A=@n6, B=@n7))
  n9 = Multiply(A=@n6, B=@n7)
  n10 = LinearInterpolate(A=@n5, B=Multiply(A=@n5) {constB=-1}, Alpha=Saturate(None=Subtract(A=@n8, B=@n9)))
  n11 = Multiply(
   A=Frac(None=Multiply(A=@n4, B=@n10) {constB=0.1})
   B=LinearInterpolate(A=Constant() {r=0.49}, B=Constant() {r=0.46}, Alpha=NamedRerouteUsage()) {constA=0.5, constB=0.54}
  ) {constB=0.5}
  n12 = Add(A=ComponentMask(None=Multiply(A=@n1, B=@n3)) {g=true}, B=@n11)
  n13 = NamedRerouteUsage()
  n14 = Constant() {r=0.501}
  n15 = Constant() {desc=0.535, r=0.535}
  n16 = NamedRerouteUsage()
  n17 = Frac(None=Add(A=@n11, B=LinearInterpolate(A=@n14, B=@n15, Alpha=@n16) {constA=0.5, constB=0.54}) {constB=0.5})
  n18 = Add(A=Step(Y=@n12, X=Constant() {r=0.5}) {constX=0.505, constY=0.5125}, B=Step(Y=Constant() {r=1}, X=@n12) {constY=1, desc=BottomMask})
  n19 = Add(A=Step(Y=@n12) {constX=0.5125, constY=0.5125}, B=Step(Y=Constant() {desc=0.977, r=0.974}, X=@n12) {constY=1})
  n20 = NamedRerouteUsage()
  n21 = MaterialFunctionCall(
   X=NamedRerouteUsage()
   Y=LinearInterpolate(A=@n12, B=Add(A=@n13, B=@n17) {constB=0.5}, Alpha=LinearInterpolate(A=@n18, B=@n19, Alpha=@n20) {constA=1, constB=0.975})
  ) {materialFunction=&MakeFloat2}
  n22 = ChannelMaskParameter(None=NamedRerouteUsage()) {group=TankTreadPanning, maskChannel=Blue, parameterName=TreadsChannel, defaultValue=<rgba 0 0 1 0>}
  n23 = MaterialFunctionCall(Overide Texture Set UVs=LinearInterpolate(A=TextureCoordinate(), B=@n21.Result, Alpha=@n22.R)) {materialFunction=&MF_CharacterParent_2023_PreFX}
  MaterialFunctionCall(Material=@n23.Result, M=@n23.Result, Pre Skinned Local Position=@n23.Result, Distance Blend=@n23.Result, Fresnel=@n23.Result) {materialFunction=&MF_CharacterParent_2023_PostFX}
  n24 = LocalPosition() {localOrigin=InstancePreSkinning}
  n25 = ShaderStageSwitch(PixelShader=VertexInterpolator(VS=@n24.XYZ).PS, VertexShader=@n24.XYZ)
  NamedRerouteDeclaration(None=Step(Y=ComponentMask(None=@n25) {g=true}) {constX=0}) {name=RearMask, nodeColor=<rgba 0 1 0.7891 1>}
  NamedRerouteDeclaration(None=@n2.R) {name=R, nodeColor=<rgba 1 0.4688 0 1>}
  NamedRerouteDeclaration(None=@n23.Tech Art Mask) {name=TechArtMask, nodeColor=<rgba 0.1719 1 0 1>}
  NamedRerouteDeclaration(None=Step(Y=NamedRerouteUsage()) {constX=0.5}) {name=InnerTreadsMask, nodeColor=<rgba 0 1 0.2266 1>}
  OneMinus(None=Floor(None=ScalarParameter() {group=TankTreadPanning, parameterName=TreadsStaticLerping, sortPriority=0}))
  TextureSample(UVs=@n21.Result) {texture=/BRCosmetics/Animation/Props/Gadgets_KnitBarrel/Textures/T_KnitBarrel_Emote_Prop_D}
 ]
}
