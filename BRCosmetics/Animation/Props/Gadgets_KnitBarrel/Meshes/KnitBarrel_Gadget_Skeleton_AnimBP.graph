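# Animation Blueprint for the KnitBarrel gadget prop skeleton, switching animation and driving a material by the prop's movement.
# The AnimGraph blends between the InMotion slot while bIsMoving2D is true and DefaultSlot otherwise.
# Shortly after BeginPlay, it caches a dynamic material from the owning FortPlayerPawn's first emote mesh component.
# Each update while moving, it accumulates delta time and writes it to that material's TimeElasped parameter.
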AnimBlueprint /BRCosmetics/Animation/Props/Gadgets_KnitBarrel/Meshes/KnitBarrel_Gadget_Skeleton_AnimBP
target_skeleton=&/KnitBarrel_Gadget_Skeleton
graphs [
 AnimGraph {nodes [Output Pose(Result=Blend Poses by bool(BlendPose_0=Slot 'InMotion'() {slotName=InMotion}, BlendPose_1=Slot 'DefaultSlot'(), bActiveValue=Get bIsMoving2D()))]}
 EventGraph {nodes [
   n1 = Event BlueprintUpdateAnimation()
   > Branch(Condition=Get bIsMoving2D())
   > Set TimeElasped(TimeElasped:double=float + float(A=@n1.DeltaTimeX, B=Get TimeElasped() -> (TimeElasped:double))) -> (Output_Get:double)
   n2 = Get MID() -> (MID:Material Instance Dynamic Object Reference)
   > Is Valid(InputObject=@n2)
   > SetScalarParameterValue(self=@n2, ParameterName=TimeElasped, Value=Get TimeElasped())
   Event BlueprintBeginPlay() {is=title K2Node_Event}
   > Parent: BlueprintBeginPlay() {is=title K2Node_CallParentFunction}
   > Delay(Duration=0.1)
   > n3 = Cast To FortPlayerPawn(Object=TryGetPawnOwner())
   > n4 = Values!(
    execute
    TargetMap:Map of Integers to Emote Mesh Comp Data Structures=Get EmoteMeshCompDatas!(self:Fort Pawn Object Reference=@n3) -> (EmoteMeshCompDatas:Map of Integers to Emote Mesh Comp Data Structures)
   ) -> (then, Values:Array of Emote Mesh Comp Data Structures)
   n5 = "Get (a ref)"(Array:Array of Emote Mesh Comp Data Structures=@n4, Dimension 1=0) -> (Output:"Emote Mesh Comp Data Structure (by ref)")
   > Set MID(MID:Material Instance Dynamic Object Reference=GetDynamicMaterialInstance!(
     MeshComponent:Mesh Component Object Reference=Break Emote Mesh Comp Data!(EmoteMeshCompData:"Emote Mesh Comp Data Structure (by ref)"=@n5) -> (MeshComp:Mesh Component Object Reference, PropExclusionTag:Gameplay Tag Structure, bTickEvenWhenNotRendered:bool=~) {is=title K2Node_BreakStruct}
     MaterialIndex:int=0
    ) -> (Material Instance Dynamic:Material Instance Dynamic Object Reference)) -> (Output_Get:Material Instance Dynamic Object Reference)
  ]}
]
variables [
 MID {type=MaterialInstanceDynamic}
 TimeElasped {type=number}
]
