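# Physics asset for the KnitBarrel gadget cosmetic prop, giving its skeletal mesh simple collision bodies and joints.
# Each bone gets one capsule: fitted-size shapes for the container, egg, plate and Skyhopper joints, tiny ones for vehicle and main, and a very large-radius one on root.
# Constraints form a star, linking every JNT bone to main and main to root, all sharing identical swing and twist limit angles.
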PhysicsAsset /BRCosmetics/Animation/Props/Gadgets_KnitBarrel/Meshes/PA_KnitBarrel_Gadget
bodies {
 JNT_Containter [{shapeType=Capsule, length=17.0183, radius=15.5778, center=<0.0001 0.2405 3.1953>, rotation=<rot 0 179.9805 -178.795>}]
 JNT_Egg [{shapeType=Capsule, length=16.7299, radius=12.0805, center=<0 0.0351 3.6773>, rotation=<rot 0 -90 0>}]
 JNT_Plate [{shapeType=Capsule, length=21.5121, radius=10.685, center=<0.7737 0.4024 20.3043>, rotation=<rot 2.3991 -90 4.5418>}]
 JNT_Skyhopper [{shapeType=Capsule, length=21.5688, radius=17.0126, center=<-0.0035 0.5448 3.7676>, rotation=<rot 40.4261 -90 -0.0241>}]
 JNT_Vehicle [{shapeType=Capsule, length=0.505, radius=0.505, rotation=<rot 0 -90 0>}]
 main [{shapeType=Capsule, length=0.505, radius=0.505, rotation=<rot 0 -90 0>}]
 root [{shapeType=Capsule, length=0.505, radius=309.2942, rotation=<rot 0 -90 0>}]
}
constraints [
 {bone1Name=main, bone2Name=root, swing1LimitDegrees=45, swing2LimitDegrees=45, twistLimitDegrees=45}
 {bone1Name=JNT_Vehicle, bone2Name=main, swing1LimitDegrees=45, swing2LimitDegrees=45, twistLimitDegrees=45}
 {bone1Name=JNT_Skyhopper, bone2Name=main, swing1LimitDegrees=45, swing2LimitDegrees=45, twistLimitDegrees=45}
 {bone1Name=JNT_Egg, bone2Name=main, swing1LimitDegrees=45, swing2LimitDegrees=45, twistLimitDegrees=45}
 {bone1Name=JNT_Containter, bone2Name=main, swing1LimitDegrees=45, swing2LimitDegrees=45, twistLimitDegrees=45}
 {bone1Name=JNT_Plate, bone2Name=main, swing1LimitDegrees=45, swing2LimitDegrees=45, twistLimitDegrees=45}
]
